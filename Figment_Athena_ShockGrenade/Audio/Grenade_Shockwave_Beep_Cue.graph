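# Sound cue for the beep of the Athena shock grenade gadget in the Figment plugin.
# A single wave player feeds one beep wave straight to the output, with no randomization or modulation.
# It routes through the Fort gadgets sound class with gadget-activation attenuation and a reduced volume.
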SoundCue /Figment_Athena_ShockGrenade/Audio/Grenade_Shockwave_Beep_Cue
root=WavePlayer_0
nodes [{id=WavePlayer_0, properties {soundWaveAssetPtr=&/Grenade_Shockwave_Beep01}}]
properties {
 attenuationSettings=&Fort_GadgetActivate_Att
 maxDistance=4000
 soundClassObject=&Fort_Gadgets
 volumeMultiplier=0.6
}
