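# Sound cue for the Figment shockwave grenade's stick sound.
# It picks one of three equally weighted stick waves at random, then an Enveloper lowers the pitch, slightly varies pitch and volume, and fades it out in under half a second.
# It plays in the ranged-weapon projectile effects sound class, with overridden attenuation that adds a distance-based low-pass filter.
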SoundCue /Figment_Athena_ShockGrenade/Audio/Grenade_Shockwave_Stick_Figment_Cue
root=Enveloper_0
nodes [
 {
  id=Enveloper_0
  children [Random_1]
  properties {
   pitchMax=0.8
   pitchMin=0.75
   volumeMax=1.05
   volumeMin=0.95
   pitchCurve {editorCurveData {keys [{time=0}]}}
   volumeCurve {editorCurveData {keys [
      {interpMode=RCIM_Cubic, tangentMode=RCTM_User, time=0, value=1}
      {arriveTangent=-2.1674, interpMode=RCIM_Cubic, leaveTangent=-2.1674, tangentMode=RCTM_User, time=0.2, value=0.5}
      <at 0.38 0>
     ]}}
  }
 }
 {id=Random_1, children [WavePlayer_1,WavePlayer_2,WavePlayer_3], properties {weights [1,1,1]}}
 {id=WavePlayer_1, properties {soundWaveAssetPtr=&/Grenade_Shockwave_Stick01}}
 {id=WavePlayer_2, properties {soundWaveAssetPtr=&/Grenade_Shockwave_Stick02}}
 {id=WavePlayer_3, properties {soundWaveAssetPtr=&/Grenade_Shockwave_Stick03}}
]
properties {
 attenuationSettings=/Game/Sounds/Attenuation/Fort_Grenade_Bounce_Att
 bOverrideAttenuation=true
 maxDistance=3000
 soundClassObject=&Fort_WeaponRanged_Projectile_Effects
 volumeMultiplier=0.372
 attenuationOverrides {bAttenuateWithLPF=true, falloffDistance=2500, lPFFrequencyAtMax=2000, lPFRadiusMax=3000, lPFRadiusMin=500, attenuationShapeExtents {x=500}}
}
