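# Sound cue for throwing the Figment shockwave grenade, classed as player foley with basic foley attenuation.
# A random node picks one of three equally weighted throw waves, then a modulator at default settings processes it.
# A quality-level root sends its first three slots to that modulator and leaves the fourth empty.
# The cue overrides concurrency to block new instances once the limit is reached.
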SoundCue /Figment_Athena_ShockGrenade/Audio/Grenade_Shockwave_Throw_Figment_Cue
root=QualityLevel_0
nodes [
 {id=Modulator_0, children [Random_1]}
 {id=QualityLevel_0, children [Modulator_0,Modulator_0,Modulator_0,null]}
 {id=Random_1, children [WavePlayer_0,WavePlayer_1,WavePlayer_2], properties {bShouldExcludeFromBranchCulling=true, weights [1,1,1]}}
 {id=WavePlayer_0, properties {soundWaveAssetPtr=&/Grenade_Shockwave_Throw03}}
 {id=WavePlayer_1, properties {soundWaveAssetPtr=&/Grenade_Shockwave_Throw02}}
 {id=WavePlayer_2, properties {soundWaveAssetPtr=&/Grenade_Shockwave_Throw01}}
]
properties {
 attenuationSettings=&Foley_Attenuation_Basic_Att
 bOverrideConcurrency=true
 maxDistance=1536
 soundClassObject=&Fort_Player_Foley
 volumeMultiplier=1.2797
 concurrencyOverrides {resolutionRule=PreventNew}
}
